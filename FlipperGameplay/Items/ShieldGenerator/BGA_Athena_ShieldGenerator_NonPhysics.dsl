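; Non-physics Shield Generator item actor that supplies dynamic destruction settings, replicated FX toggling and server-side marking.
; It reports its static mesh as Metal for dynamic destruction and opts into that system's material and Niagara effects.
; The SetFXMode rep-notify plays or ends FX, ignoring negative values.
; TryMarkOnServer follows the given pawn's convert component to its converter pawn, then marks this actor through that controller's marker component.
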(fn GetDynamicDestructionResourceType
  (return "Metal"))

(fn GetDynamicDestructionStaticMeshComponents
  (return (Utilities|Array|MakeArray (Variables|Default|GetStaticMesh)) 0 0))

(fn OnRep_SetFXMode
  (if (>= (Variables|Default|GetSetFXMode) 0)
    (if (>= (Variables|Default|GetSetFXMode) 1)
      (CallFunction|PlayFX)
      (else
        (CallFunction|EndFX)))))

(fn ShouldUseDynamicDestructionMaterial
  (return true))

(fn ShouldUseDynamicDestructionNiagaraSystem
  (return true))

(fn TryMarkOnServer
  (params (PlayerPawn :type Fort%20Player%20Pawn%20Object%20Reference))
  (bind _convertcomponent (Variables|/Script/FortniteGame.FortPlayerPawn|GetConvertComponent PlayerPawn))
  (bind _returnvalue (Convert|GetConverterPawn _convertcomponent))
  (bind _returnvalue_2 (Fortnite|GetFortPlayerControllerfromActor _returnvalue))
  (bind _returnvalue_3 (Class|FortPlayerController|GetMarkerComponent _returnvalue_2))
  (Utilities|IsValid _convertcomponent
    (:Is%20Valid
      (Utilities|IsValid _returnvalue
        (:Is%20Valid
          (Utilities|IsValid _returnvalue_2
            (:Is%20Valid
              (Utilities|IsValid _returnvalue_3
                (:Is%20Valid
                  (MarkerComponent|MarkActoronServer _returnvalue_3 self true true))))))))))

(fn ConstructionScript
  (|CallParentFunction))
