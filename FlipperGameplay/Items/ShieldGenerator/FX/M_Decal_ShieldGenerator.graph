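# Unlit translucent deferred decal material for the Shield Generator item, drawing a blue, water-rippled circular effect on surfaces.
# Layered water normals, distorted over sphere-projected UVs, supply a blue channel that blends ColorA and ColorB for emissive.
# Sphere masks broken up by smoke noise shape opacity, scaled by decal lifetime fade.
# Shading-path and platform switches provide vertex-interpolated UV normals or flat fallbacks for cheaper targets.
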Material /FlipperGameplay/Items/ShieldGenerator/FX/M_Decal_ShieldGenerator
graph {
 properties {
  blendMode=BLEND_Translucent
  materialDomain=MD_DeferredDecal
  pixelDepthOffsetMode=PDOM_Legacy
  shadingModel=MSM_Unlit
  usedShadingModels=Unlit
 }
 nodes [
  Constant() {r=0.5}
  n1 = ScalarParameter() {defaultValue=0.33, parameterName=Low_Roughness}
  n2 = ComponentMask(None=VectorParameter() {group=Normals, parameterName=NormalUVScale, defaultValue=<rgba 1 0.381 0 0>}.RGB) {g=true, r=true}
  n3 = MaterialFunctionCall!(UVs=TextureCoordinate(), Additional Scale) {materialFunction=/Game/Effects/Fort_Effects/Materials/_Functions/MF_DecalSphereCoords}
  n4 = TextureSample(UVs=Add(A=Multiply(A=@n3.Result, B=@n2) {constB=0.8556}, B=ScalarParameter() {defaultValue=0.2, parameterName=B_1}) {constB=0.3215}) {samplerType=SAMPLERTYPE_Normal, texture=/Game/Global/Textures/Default/Tileable_Patterns/T_Water2_N}
  n5 = Multiply(A=ComponentMask(None=@n4.RGB) {g=true, r=true}, B=ScalarParameter() {defaultValue=0.1, group=Normals, parameterName=Noise 1}) {constB=0.04}
  n6 = Multiply(
   A=@n3.Result
   B=ComponentMask(None=VectorParameter() {group=Normals, parameterName=Normal2UVScale, defaultValue=<rgba 1 0.4667 0 0>}.RGB) {g=true, r=true}
  )
  n7 = MaterialFunctionCall(
   BaseNormal=@n4.RGB
   AdditionalNormal=TextureSample(UVs=Add(A=@n5, B=@n6)) {samplerType=SAMPLERTYPE_Normal, texture=&T_Water2_N}.RGB
  ) {materialFunction=&BlendAngleCorrectedNormals}
  n8 = TextureCoordinate()
  n9 = VectorParameter() {group=Normals_Mobile, parameterName=NormalUVScale_M3, sortPriority=1, defaultValue=<rgba 1 0.4667 0 0>}
  n10 = ComponentMask(None=@n9.RGB) {g=true, r=true}
  n11 = TextureSample!(UVs=VertexInterpolator(VS=Multiply(A=@n8, B=@n10)).PS, Tex, Level, Apply View MipBias) {mipValueMode=TMVM_MipLevel, samplerType=SAMPLERTYPE_Normal, texture=/Game/Global/Textures/Default/Tileable_Patterns/T_WaterSoft_N}
  n12 = Multiply(
   A=TextureCoordinate()
   B=ComponentMask(None=VectorParameter() {group=Normals_Mobile, parameterName=NormalUVScale_M/M2, defaultValue=<rgba 1 0.381 0 0>}.RGB) {g=true, r=true}
  ) {constB=0.8556}
  n13 = TextureSample(UVs=VertexInterpolator(VS=@n12).PS) {mipValueMode=TMVM_MipLevel, samplerType=SAMPLERTYPE_Normal, texture=&T_Water2_N}
  n14 = TextureSample(UVs=VertexInterpolator(VS=@n12).PS) {mipValueMode=TMVM_MipLevel, samplerType=SAMPLERTYPE_Normal, texture=&T_Water2_N}
  n15 = MaterialFunctionCall(
   BaseNormal=MaterialFunctionCall(Normal=@n11.RGB, Flatness=ScalarParameter() {defaultValue=0.9115, group=Normals_Mobile, parameterName=NormalFlaten_M3, sortPriority=5}) {materialFunction=&FlattenNormal}.Result
   AdditionalNormal=MaterialFunctionCall(
    Normal=MaterialFunctionCall(BaseNormal=@n13.RGB, AdditionalNormal=@n14.RGB) {materialFunction=&BlendAngleCorrectedNormals}.Result
    Flatness=ScalarParameter() {group=Normals, parameterName=NormalFlatten}
   ) {materialFunction=&FlattenNormal}.Result
  ) {materialFunction=&BlendAngleCorrectedNormals}
  n16 = ShadingPathSwitch(
   Default=MaterialFunctionCall(Normal=Normalize(VectorInput=@n7.Result), Flatness=ScalarParameter() {group=Normals, parameterName=NormalFlatten}) {materialFunction=&FlattenNormal}.Result
   Forward=@n15.Result
   Mobile=@n15.Result
  )
  n17 = ComponentMask(None=@n16) {b=true}
  n18 = LinearInterpolate(Alpha=@n17) {constA=0.1, constB=0.07}
  MaterialFunctionCall("Xb1,PS4"=@n18, "PC-High&&Epic,XSX,PS5"=@n18, "PC-Perf,Android"=@n1, Switch=@n1, "PC-Low(NotPerf)&&Medium"=@n1) {materialFunction=&PlatformSwitch_1D}
  n19 = Constant3Vector() {constant=<rgba 0 0 1 0>}
  MaterialFunctionCall!("Xb1,PS4"=@n16, "PC-High&&Epic,XSX,PS5"=@n16, "PC-Perf,Android"=@n19, Switch=@n19, "PC-Low(NotPerf)&&Medium"=@n19) {materialFunction=/Game/Effects/Niagara/MasterMaterials/Functions/Common/PlatformSwitch_3D}
  n20 = VectorParameter() {group=Color, parameterName=ColorA, defaultValue=<rgba 0 0.7 1 1>}
  n21 = VectorParameter() {group=Color, parameterName=ColorB, defaultValue=<rgba 0.0074 0.2465 0.349 1>}
  n22 = Saturate(None=Power(Base=@n17, Exp=Constant() {r=10}))
  MaterialFunctionCall(Emissive=Multiply(A=LinearInterpolate(A=@n20.RGB, B=@n21.RGB, Alpha=@n22), B=ScalarParameter() {defaultValue=0.5, parameterName=Emissive}) {constB=0}) {materialFunction=&MF_FX_EmissiveAdjustments}
  n23 = Multiply(
   A=TextureCoordinate()
   B=ComponentMask(None=VectorParameter() {parameterName=Noise UV Scale, defaultValue=<rgba 2 2 0 0>}.RGB) {g=true, r=true}
  )
  n24 = ComponentMask(None=VectorParameter() {parameterName=Noise Offset, defaultValue=<rgba 0.0533 0 0 0>}.RGB) {g=true, r=true}
  n25 = Power(
   Base=TextureSample!(UVs=Add(A=@n23, B=@n24), Tex, Apply View MipBias) {texture=&T_Inky_Smoke_Tile}.R
   Exp=ScalarParameter() {defaultValue=-0.05, parameterName=Exp}
  )
  n26 = SphereMask(
   A=TextureCoordinate()
   B=Constant() {r=0.5}
   Radius=ScalarParameter() {defaultValue=0.5, parameterName=RadiusInner}
   Hardness=ScalarParameter() {defaultValue=0.5, parameterName=Hardness}
  )
  n27 = OneMinus(None=Multiply(A=@n25, B=OneMinus(None=@n26)))
  n28 = TextureCoordinate()
  n29 = Constant() {r=0.5}
  n30 = ScalarParameter() {defaultValue=0.5, parameterName=RadiusB}
  n31 = ScalarParameter() {defaultValue=100, parameterName=HardnessB}
  n32 = SphereMask(
   A=TextureCoordinate()
   B=Constant() {r=0.5}
   Radius=ScalarParameter() {defaultValue=0.0947, parameterName=Radius}
   Hardness=ScalarParameter() {defaultValue=0.5, parameterName=Hardness}
  )
  n33 = Multiply(A=Subtract(A=SphereMask(A=@n28, B=@n29, Radius=@n30, Hardness=@n31), B=Multiply(A=@n32, B=@n25)), B=@n27)
  n34 = Multiply(A=Multiply(A=DecalLifetimeOpacity().Opacity, B=Saturate(None=@n33)), B=ScalarParameter() {defaultValue=0.35, parameterName=Opacity})
  NamedRerouteDeclaration(None=@n34) {name=Opacity, nodeColor=<rgba 1 0 0.1406 1>}
  NamedRerouteUsage()
  Reroute(None=LinearInterpolate(Alpha=@n17) {constB=0.5})
  Saturate(None=@n34)
  ShadingPathSwitch(Default=@n27, Mobile=QualitySwitch(Default=@n27, Low=Saturate(None=@n26)))
 ]
}
