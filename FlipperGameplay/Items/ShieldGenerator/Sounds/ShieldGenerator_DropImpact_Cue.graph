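# Sound cue for the Shield Generator item's drop impact, classed as an impact sound.
# A Random node picks one of three equally weighted ShieldSprinkler drop impact waves and feeds it through a root Modulator left at default settings.
# The cue raises pitch, lowers volume, and borrows shared small-object demolition hit attenuation and a four-voice concurrency rule that stops the furthest or oldest.
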SoundCue /FlipperGameplay/Items/ShieldGenerator/Sounds/ShieldGenerator_DropImpact_Cue
root=Modulator_0
nodes [
 {id=Modulator_0, children [Random_1]}
 {id=Random_1, children [WavePlayer_10, WavePlayer_11, WavePlayer_12], properties {weights [1,1,1]}}
 {id=WavePlayer_10, properties {soundWaveAssetPtr=&/ShieldSprinkler_DropImpact_01}}
 {id=WavePlayer_11, properties {soundWaveAssetPtr=&/ShieldSprinkler_DropImpact_02}}
 {id=WavePlayer_12, properties {soundWaveAssetPtr=&/ShieldSprinkler_DropImpact_03}}
]
properties {
 attenuationSettings=/Game/Sounds/Attenuation/Fort_Demolition_Hit_Small_Sized_Object_Att
 maxDistance=4000
 pitchMultiplier=1.2
 soundClassObject=&Fort_Impacts
 volumeMultiplier=0.7
 concurrencyOverrides {maxCount=4}
 concurrencySet [/Game/Sounds/Concurrency_Rules/Demo_FourMaximum_StopFurthestOldest_VolumeScale]
}
