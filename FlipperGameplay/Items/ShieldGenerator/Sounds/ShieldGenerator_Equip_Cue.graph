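# Sound cue played for equipping the Shield Generator item.
# A Random node picks one of three ShieldKeg equip waves with equal weight and feeds it through a Modulator at the root.
# The cue plays quieter and higher-pitched, uses the weapon reload attenuation and loot-collect sound class, and overrides concurrency to cap simultaneous instances.
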SoundCue /FlipperGameplay/Items/ShieldGenerator/Sounds/ShieldGenerator_Equip_Cue
root=Modulator_0
nodes [
 {id=Modulator_0, children [Random_0]}
 {id=Random_0, children [WavePlayer_1,WavePlayer_4,WavePlayer_5], properties {weights [1,1,1]}}
 {id=WavePlayer_1, properties {soundWaveAssetPtr=&/ShieldKeg_Equip_03}}
 {id=WavePlayer_4, properties {soundWaveAssetPtr=&/ShieldKeg_Equip_02}}
 {id=WavePlayer_5, properties {soundWaveAssetPtr=&/ShieldKeg_Equip_01}}
]
properties {
 attenuationSettings=&Fort_Weapon_Reload_Att
 bOverrideConcurrency=true
 maxDistance=2250
 pitchMultiplier=1.2
 soundClassObject=&Fort_Loot_Collect
 volumeMultiplier=0.4
 concurrencyOverrides {maxCount=3}
}
