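# Looping Sound Cue for the eruption of the Shield Generator item, routed to the water interaction sound class.
# A mixer blends the combined eruption loop with a quieter extra layer that a modulator pitches up by a fixed amount.
# A root modulator at default settings applies slight variation to the mix.
# Overridden attenuation gives logarithmic falloff, a non-spatialized near radius and occlusion that lowers volume and applies a low-pass filter.
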SoundCue /FlipperGameplay/Items/ShieldGenerator/Sounds/ShieldGenerator_Erupt_Loop_Cue
root=Modulator_0
nodes [
 {id=Mixer_1, children [WavePlayer_4,Modulator_3], properties {inputVolume [1,0.45]}}
 {id=Modulator_0, children [Mixer_1]}
 {id=Modulator_3, children [WavePlayer_7], properties {pitchMax=1.1, pitchMin=1.1}}
 {
  id=WavePlayer_4
  properties {bLooping=true, soundWaveAssetPtr=&/ShieldSprinkler_EruptLoopCombined}
 }
 {
  id=WavePlayer_7
  properties {bLooping=true, soundWaveAssetPtr=&/ShieldSprinkler_EruptLoopLayer_04}
 }
]
properties {
 attenuationSettings=/Game/Sounds/Attenuation/Fort_Player_Swimming_Athena_Att
 bOverrideAttenuation=true
 maxDistance=2200
 pitchMultiplier=1.1
 soundClassObject=/Game/Sounds/FortSoundClassesAndModes/Fort_Water_Interaction
 volumeMultiplier=0.8
 attenuationOverrides {
  bEnableOcclusion=true
  distanceAlgorithm=Logarithmic
  falloffDistance=1800
  nonSpatializedRadiusStart=600
  occlusionLowPassFilterFrequency=18000
  occlusionTraceChannel=ECC_GameTraceChannel11
  occlusionVolumeAttenuation=0.6
  stereoSpread=2000
 }
}
