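# Sound cue for the start of the Shield Generator item's eruption.
# It mixes a fixed main wave with a slightly quieter layer chosen at random from three equally weighted variants and pitched up, then runs the mix through an outer modulator.
# Cue-level overrides add spatial attenuation with occlusion filtering and cap concurrency at a few voices, stopping the oldest.
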SoundCue /FlipperGameplay/Items/ShieldGenerator/Sounds/ShieldGenerator_Erupt_Start_Cue
root=Modulator_0
nodes [
 {id=Mixer_1, children [WavePlayer_12, Modulator_1], properties {inputVolume [1,0.8]}}
 {id=Modulator_0, children [Mixer_1]}
 {id=Modulator_1, children [Random_4], properties {pitchMax=1.2, pitchMin=1.2}}
 {id=Random_4, children [WavePlayer_17, WavePlayer_18, WavePlayer_19], properties {weights [1,1,1]}}
 {id=WavePlayer_12, properties {soundWaveAssetPtr=&/ShieldSprinkler_EruptStart_04}}
 {id=WavePlayer_17, properties {soundWaveAssetPtr=&/ShieldSprinkler_EruptStartLayer_02}}
 {id=WavePlayer_18, properties {soundWaveAssetPtr=&/ShieldSprinkler_EruptStartLayer_03}}
 {id=WavePlayer_19, properties {soundWaveAssetPtr=&/ShieldSprinkler_EruptStartLayer_01}}
]
properties {
 attenuationSettings=/Game/Sounds/Attenuation/Fort_Bullet_Impact_NPC_Weapon_Athena_Att
 bOverrideAttenuation=true
 bOverrideConcurrency=true
 maxDistance=1850
 pitchMultiplier=1.15
 soundClassObject=&Fort_WeaponRanged_Fire_3P_PSM
 volumeMultiplier=1
 attenuationOverrides {
  bEnableOcclusion=true
  falloffDistance=1800
  occlusionLowPassFilterFrequency=18000
  occlusionTraceChannel=ECC_GameTraceChannel11
  occlusionVolumeAttenuation=0.8
  stereoSpread=2000
  attenuationShapeExtents {x=50}
 }
 concurrencyOverrides {maxCount=3, resolutionRule=StopOldest}
}
