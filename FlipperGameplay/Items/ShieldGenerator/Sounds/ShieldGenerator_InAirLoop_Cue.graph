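# Sound cue for the Shield Generator's in-air loop, by name the sound heard while the item is airborne.
# A Random node picks one of two looping ShieldSprinkler waves, and a Modulator applies slight pitch variation.
# It borrows SpicySoda in-air attenuation and uses a third-person ranged weapon fire sound class.
# Overridden concurrency caps simultaneous copies, stopping the farthest first, and applies volume scaling with a gradual release.
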SoundCue /FlipperGameplay/Items/ShieldGenerator/Sounds/ShieldGenerator_InAirLoop_Cue
root=Modulator_0
nodes [
 {id=Modulator_0, children [Random_0], properties {pitchMax=1.02, pitchMin=0.98}}
 {id=Random_0, children [WavePlayer_0,WavePlayer_1], properties {weights [1,1]}}
 {id=WavePlayer_0, properties {bLooping=true, soundWaveAssetPtr=&/ShieldSprinkler_InAirLoop_01}}
 {id=WavePlayer_1, properties {bLooping=true, soundWaveAssetPtr=&/ShieldSprinkler_InAirLoop_02}}
]
properties {
 attenuationSettings=/Game/Sounds/Attenuation/Fort_SpicySoda_InAir_Att
 bOverrideConcurrency=true
 maxDistance=2500
 soundClassObject=&Fort_WeaponRanged_Fire_3P_PSM
 volumeMultiplier=1
 attenuationOverrides {falloffDistance=5000, attenuationShapeExtents {x=1000}}
 concurrencyOverrides {
  bVolumeScaleCanRelease=true
  maxCount=3
  resolutionRule=StopFarthestThenPreventNew
  volumeScale=0.2
  volumeScaleAttackTime=0.2
  volumeScaleReleaseTime=2
 }
 concurrencySet [/Game/Sounds/Concurrency_Rules/Athena_Explosive_Fuse_Conc]
}
