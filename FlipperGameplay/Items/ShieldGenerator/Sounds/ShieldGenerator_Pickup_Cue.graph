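# Sound cue for the Shield Generator item's pickup sound in FlipperGameplay.
# A random node picks one of three equally weighted ShieldSprinkler pickup waves and feeds it through a modulator left at its usual settings.
# The cue uses weapon-pickup attenuation, the gadget items sound class and reduced volume.
# Its concurrency override blocks new instances once the limit is reached.
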SoundCue /FlipperGameplay/Items/ShieldGenerator/Sounds/ShieldGenerator_Pickup_Cue
root=Modulator_0
nodes [
 {id=Modulator_0, children [Random_0]}
 {id=Random_0, children [WavePlayer_0,WavePlayer_2,WavePlayer_3], properties {weights [1,1,1]}}
 {id=WavePlayer_0, properties {soundWaveAssetPtr=&/ShieldSprinkler_Pickup_03}}
 {id=WavePlayer_2, properties {soundWaveAssetPtr=&/ShieldSprinkler_Pickup_01}}
 {id=WavePlayer_3, properties {soundWaveAssetPtr=&/ShieldSprinkler_Pickup_02}}
]
properties {
 attenuationSettings=&Fort_Weapon_Pickup_Att
 bOverrideConcurrency=true
 maxDistance=3500
 soundClassObject=&Fort_Gadgets_Items
 volumeMultiplier=0.3
 concurrencyOverrides {resolutionRule=PreventNew}
}
