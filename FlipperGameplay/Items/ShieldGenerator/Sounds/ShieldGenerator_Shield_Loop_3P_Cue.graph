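# Sound cue for the Shield Generator's shield loop, named as the third-person (3P) version.
# A mixer layers a looping shield wave with a non-looping start wave, which a modulator pitches slightly up.
# An enveloper volume curve ramps the loop in over roughly the first third of a second, under the start sound.
# It uses non-occluded consumable attenuation and a concurrency override allowing a single instance.
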SoundCue /FlipperGameplay/Items/ShieldGenerator/Sounds/ShieldGenerator_Shield_Loop_3P_Cue
root=Mixer_0
nodes [
 {
  id=Enveloper_0
  children [WavePlayer_0]
  properties {pitchCurve {editorCurveData {keys [{time=0}]}}, volumeCurve {editorCurveData {keys [<at -0.0029 -14.8582>,<at 0.3177 0.4118>]}}}
 }
 {id=Mixer_0, children [Modulator_0,Modulator_1], properties {inputVolume [1,1]}}
 {id=Modulator_0, children [Enveloper_0], properties {pitchMax=1, pitchMin=1}}
 {id=Modulator_1, children [WavePlayer_2], properties {pitchMax=1.1, pitchMin=1.1}}
 {id=WavePlayer_0, properties {bLooping=true, soundWaveAssetPtr=&/ShieldGenerator_Shield_Loop_01}}
 {id=WavePlayer_2, properties {soundWaveAssetPtr=&/ShieldGenerator_Shield_Start_02}}
]
properties {
 attenuationSettings=/Game/Sounds/Attenuation/Fort_Consumable_Amb_NonOccluded_Att
 bOverrideConcurrency=true
 maxDistance=1750
 soundClassObject=&Fort_Consumables
 volumeMultiplier=1
 attenuationOverrides {distanceAlgorithm=Custom, falloffDistance=10112.3027, attenuationShapeExtents {x=600}}
 concurrencyOverrides {maxCount=1}
}
