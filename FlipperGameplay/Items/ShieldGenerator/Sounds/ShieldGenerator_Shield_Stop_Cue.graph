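# Sound cue for the Shield Generator item that, by its name, plays when the generator's shield stops.
# A single wave player feeds one stop sound straight to the output, with a slight volume boost.
# It uses weapon pickup attenuation and the consumables sound class.
# A concurrency override limits simultaneous instances to a few, stopping the oldest first.
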SoundCue /FlipperGameplay/Items/ShieldGenerator/Sounds/ShieldGenerator_Shield_Stop_Cue
root=WavePlayer_1
nodes [{id=WavePlayer_1, properties {soundWaveAssetPtr=&/ShieldGenerator_Shield_Stop_02}}]
properties {
 attenuationSettings=&Fort_Weapon_Pickup_Att
 bOverrideConcurrency=true
 maxDistance=3500
 soundClassObject=&Fort_Consumables
 volumeMultiplier=1.2
 concurrencyOverrides {maxCount=3, resolutionRule=StopOldest}
}
