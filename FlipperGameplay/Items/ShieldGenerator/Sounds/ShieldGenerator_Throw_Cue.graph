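# Sound cue for throwing the Shield Generator gadget item.
# A Random node picks one of three equally weighted ShieldSprinkler throw waves, and a Modulator applies slight pitch variation.
# It plays at reduced volume in the gadgets items sound class with gadget-placement attenuation.
# A concurrency override prevents new instances rather than stopping playing ones when the limit is reached.
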SoundCue /FlipperGameplay/Items/ShieldGenerator/Sounds/ShieldGenerator_Throw_Cue
root=Modulator_2
nodes [
 {id=Modulator_2, children [Random_0], properties {pitchMax=1.02, pitchMin=0.98}}
 {id=Random_0, children [WavePlayer_0,WavePlayer_2,WavePlayer_3], properties {weights [1,1,1]}}
 {id=WavePlayer_0, properties {soundWaveAssetPtr=&/ShieldSprinkler_Throw_03}}
 {id=WavePlayer_2, properties {soundWaveAssetPtr=&/ShieldSprinkler_Throw_01}}
 {id=WavePlayer_3, properties {soundWaveAssetPtr=&/ShieldSprinkler_Throw_02}}
]
properties {
 attenuationSettings=/Game/Sounds/Attenuation/Fort_GadgetPlace_Att
 bOverrideConcurrency=true
 maxDistance=2000
 soundClassObject=&Fort_Gadgets_Items
 volumeMultiplier=0.5
 concurrencyOverrides {resolutionRule=PreventNew}
}
