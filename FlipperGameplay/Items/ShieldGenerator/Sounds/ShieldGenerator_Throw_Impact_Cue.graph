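# Sound cue for the impact of a thrown Shield Generator item, routed to the Fort_Impacts sound class.
# A Random node picks one of three equally weighted ShieldSprinkler throw-impact waves, which a Modulator varies slightly in pitch.
# The cue raises overall pitch and volume, references a shared small-object hit attenuation with override values enabled, and uses a four-voice concurrency rule that stops the furthest or oldest sound.
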SoundCue /FlipperGameplay/Items/ShieldGenerator/Sounds/ShieldGenerator_Throw_Impact_Cue
root=Modulator_0
nodes [
 {id=Modulator_0, children [Random_1], properties {pitchMax=1.02, pitchMin=0.98}}
 {id=Random_1, children [WavePlayer_0,WavePlayer_1,WavePlayer_2], properties {weights [1,1,1]}}
 {id=WavePlayer_0, properties {soundWaveAssetPtr=&/ShieldSprinkler_Throw_Impact_01}}
 {id=WavePlayer_1, properties {soundWaveAssetPtr=&/ShieldSprinkler_Throw_Impact_02}}
 {id=WavePlayer_2, properties {soundWaveAssetPtr=&/ShieldSprinkler_Throw_Impact_03}}
]
properties {
 attenuationSettings=/Game/Sounds/Attenuation/Fort_Demolition_Hit_Small_Sized_Object_Att
 bOverrideAttenuation=true
 maxDistance=2600
 pitchMultiplier=1.3
 soundClassObject=&Fort_Impacts
 volumeMultiplier=1.5
 attenuationOverrides {falloffDistance=2400, stereoSpread=2000, attenuationShapeExtents {x=200}}
 concurrencyOverrides {maxCount=4}
 concurrencySet [/Game/Sounds/Concurrency_Rules/Demo_FourMaximum_StopFurthestOldest_VolumeScale]
}
