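# Sound cue for the Shield Generator item's throw-ready sound.
# It randomly picks one of two equally weighted ShieldSprinkler throw-ready waves and passes it through a modulator for slight pitch variation.
# Playback uses the weapon foley sound class and weapon reload attenuation, at reduced volume.
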SoundCue /FlipperGameplay/Items/ShieldGenerator/Sounds/ShieldGenerator_Throw_Ready_Cue
root=Modulator_0
nodes [
 {id=Modulator_0, children [Random_0], properties {pitchMax=1.02, pitchMin=0.98}}
 {id=Random_0, children [WavePlayer_0,WavePlayer_2], properties {weights [1,1]}}
 {id=WavePlayer_0, properties {soundWaveAssetPtr=&/ShieldSprinkler_ThrowReady_01}}
 {id=WavePlayer_2, properties {soundWaveAssetPtr=&/ShieldSprinkler_ThrowReady_02}}
]
properties {
 attenuationSettings=&Fort_Weapon_Reload_Att
 maxDistance=2250
 soundClassObject=&Fort_WeaponFoley
 volumeMultiplier=0.501
}
