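# Sound cue for the Shield Generator item's whip-by effect, classed as a ranged weapon projectile sound and played at low volume.
# A random node picks one of three equally weighted ShieldSprinkler whip-by waves, then a modulator slightly varies its pitch.
# Attenuation is overridden on the cue despite a referenced projectile in-air loop preset, with reverb send off and complex-collision occlusion.
# The cue's debug flag is left on.
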SoundCue /FlipperGameplay/Items/ShieldGenerator/Sounds/ShieldGenerator_WhipBy_Cue
root=Modulator_1
nodes [
 {id=Modulator_1, children [Random_4], properties {pitchMax=1.1, pitchMin=0.9}}
 {id=Random_4, children [WavePlayer_12, WavePlayer_13, WavePlayer_14], properties {weights [1,1,1]}}
 {id=WavePlayer_12, properties {soundWaveAssetPtr=&/ShieldSprinkler_WhipBy_01}}
 {id=WavePlayer_13, properties {soundWaveAssetPtr=&/ShieldSprinkler_WhipBy_02}}
 {id=WavePlayer_14, properties {soundWaveAssetPtr=&/ShieldSprinkler_WhipBy_03}}
]
properties {
 attenuationSettings=/Game/Sounds/Attenuation/Fort_Weapon_Projectile_InAir_Loop_Att
 bDebug=true
 bOverrideAttenuation=true
 maxDistance=1512
 soundClassObject=&Fort_WeaponRanged_Projectile_Effects
 volumeMultiplier=0.1
 attenuationOverrides {
  bEnableReverbSend=false
  bUseComplexCollisionForOcclusion=true
  falloffDistance=1000
  occlusionVolumeAttenuation=0
  stereoSpread=0
  attenuationShapeExtents {x=512}
 }
}
